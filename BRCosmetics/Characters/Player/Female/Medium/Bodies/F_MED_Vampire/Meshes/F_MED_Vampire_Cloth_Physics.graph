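# Physics asset for the female medium Vampire body, named as the collision set for its cloth simulation.
# Capsule bodies cover the pelvis, thighs, calves and feet, with no upper-body shapes.
# A huge zero-length capsule on the root, offset by its own radius, acts as a sphere that likely serves as a ground plane.
# Constraints chain each leg through the pelvis to the root with all angular motion left free.
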PhysicsAsset /BRCosmetics/Characters/Player/Female/Medium/Bodies/F_MED_Vampire/Meshes/F_MED_Vampire_Cloth_Physics
bodies {
 calf_l [{shapeType=Capsule, length=37.4366, radius=7.0402, center=<-20.5443 -2.1521 0.0002>, rotation=<rot 90 -1.1935 -6.1935>}]
 calf_r [{shapeType=Capsule, length=37.4366, radius=7.0402, center=<20.5443 2.1521 0.0002>, rotation=<rot 90 -7.125 167.8751>}]
 foot_l [{shapeType=Capsule, length=12.8729, radius=8.1681, center=<-3.5495 6.175 0>, rotation=<rot 0 0 90>}]
 foot_r [{shapeType=Capsule, length=12.8729, radius=8.1681, center=<3.5495 -6.175 0>, rotation=<rot 0 180 90>}]
 pelvis [{
   shapeType=Capsule
   length=11.6235
   radius=10.5052
   shapeName=pelvis_capsule
   center=<-6.7199 -1.5243 -0.2529>
   rotation=<rot 7.5237 -90 2.011>
  }]
 root [{shapeType=Capsule, length=0, radius=1000, center=<0 0 -1000>, rotation=<rot 0 -90 0>}]
 thigh_l [{shapeType=Capsule, length=30.5917, radius=7.9408, center=<-25.4763 0.1449 0.1101>, rotation=<rot 90 0 0>}]
 thigh_r [{shapeType=Capsule, length=30.5917, radius=7.9408, center=<25.4763 -0.1449 -0.1101>, rotation=<rot 90 180 0>}]
}
constraints [
 {
  bone1Name=calf_l
  bone2Name=thigh_l
  swing1LimitDegrees=45
  swing1Motion=Free
  swing2LimitDegrees=45
  swing2Motion=Free
  twistLimitDegrees=45
  twistMotion=Free
 }
 {
  bone1Name=foot_l
  bone2Name=calf_l
  swing1LimitDegrees=45
  swing1Motion=Free
  swing2LimitDegrees=45
  swing2Motion=Free
  twistLimitDegrees=45
  twistMotion=Free
 }
 {
  bone1Name=foot_r
  bone2Name=calf_r
  swing1LimitDegrees=45
  swing1Motion=Free
  swing2LimitDegrees=45
  swing2Motion=Free
  twistLimitDegrees=45
  twistMotion=Free
 }
 {
  bone1Name=calf_r
  bone2Name=thigh_r
  swing1LimitDegrees=45
  swing1Motion=Free
  swing2LimitDegrees=45
  swing2Motion=Free
  twistLimitDegrees=45
  twistMotion=Free
 }
 {
  bone1Name=thigh_l
  bone2Name=pelvis
  swing1LimitDegrees=45
  swing1Motion=Free
  swing2LimitDegrees=45
  swing2Motion=Free
  twistLimitDegrees=45
  twistMotion=Free
 }
 {
  bone1Name=thigh_r
  bone2Name=pelvis
  swing1LimitDegrees=45
  swing1Motion=Free
  swing2LimitDegrees=45
  swing2Motion=Free
  twistLimitDegrees=45
  twistMotion=Free
 }
 {
  bone1Name=pelvis
  bone2Name=root
  swing1LimitDegrees=45
  swing1Motion=Free
  swing2LimitDegrees=45
  swing2Motion=Free
  twistLimitDegrees=45
  twistMotion=Free
 }
]
