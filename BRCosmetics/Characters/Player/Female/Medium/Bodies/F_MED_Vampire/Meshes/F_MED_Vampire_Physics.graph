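# Physics asset for the F_MED_Vampire female body cosmetic, covering the legs and the dynamic skirt bones.
# Thigh and calf capsules form the leg bodies, with free rotation between each pair.
# Skirt chains A, B and C on each side run from tiny master capsules through box segments, joined by swing- and twist-limited constraints.
# The C chain is asymmetric, with two segments on the left and one on the right.
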PhysicsAsset /BRCosmetics/Characters/Player/Female/Medium/Bodies/F_MED_Vampire/Meshes/F_MED_Vampire_Physics
bodies {
 calf_l [{shapeType=Capsule, length=35.8604, radius=6.771, center=<-15.4556 -1.8434 -0.0001>, rotation=<rot 90 0 0>}]
 calf_r [{shapeType=Capsule, length=35.8604, radius=6.771, center=<15 2.3236 0>, rotation=<rot 90 0 0>}]
 dyn_skirt_A_01_l [{shapeType=Box, extentX=3.0616, extentY=10.963, extentZ=20.6933, center=<6.5442 -0.0112 0.75>, rotation=<rot 0 90 90>}]
 dyn_skirt_A_01_r [{shapeType=Box, extentX=4.223, extentY=11.6766, extentZ=16.3945, center=<11.7848 0.0635 0.5012>, rotation=<rot 0 90 90>}]
 dyn_skirt_A_02_l [{shapeType=Box, extentX=2.0444, extentY=10.9946, extentZ=19.3902, center=<6.957 -0.8767 -1.2558>, rotation=<rot 4.9238 79.9619 89.1296>}]
 dyn_skirt_A_02_r [{shapeType=Box, extentX=3.0083, extentY=7.9849, extentZ=16.9921, center=<9.3231 -1.441 0.2516>, rotation=<rot 0 90 90>}]
 dyn_skirt_A_03_l [{shapeType=Box, extentX=1.4052, extentY=6.967, extentZ=21.151, center=<9.5122 -0.7475 -0.6758>, rotation=<rot 0 90 90>}]
 dyn_skirt_A_03_r [{shapeType=Box, extentX=2.3695, extentY=5.7736, extentZ=24.898, center=<6.7472 -0.4598 0.6413>, rotation=<rot 0 90 90>}]
 dyn_skirt_A_master_l [{shapeType=Capsule, length=0.505, radius=0.505, rotation=<rot 0 -90 0>}]
 dyn_skirt_A_master_r [{shapeType=Capsule, length=0.505, radius=0.505, rotation=<rot 0 -90 0>}]
 dyn_skirt_B_01_l [{shapeType=Box, extentX=1.7825, extentY=6.8694, extentZ=13.0058, center=<6.2982 0.0366 -0.2921>, rotation=<rot 0 90 90>}]
 dyn_skirt_B_01_r [{shapeType=Box, extentX=3.0287, extentY=9.05, extentZ=11.7931, center=<6.0936 -0.258 0.5551>, rotation=<rot 0 90 90>}]
 dyn_skirt_B_02_l [{shapeType=Box, extentX=0.537, extentY=7.8533, extentZ=7.4522, center=<3.8066 0.0853 0.1664>, rotation=<rot 0 90 90>}]
 dyn_skirt_B_02_r [{shapeType=Box, extentX=1.964, extentY=3.5749, extentZ=9.9024, center=<4.5878 0.3977 -0.3064>, rotation=<rot 0 90 90>}]
 dyn_skirt_B_03_l [{shapeType=Box, extentX=1.5719, extentY=7.4785, extentZ=10.4362, center=<4.4584 0.4046 -0.1748>, rotation=<rot 0 90 90>}]
 dyn_skirt_B_03_r [{shapeType=Box, extentX=2.1344, extentY=4.6656, extentZ=9.059, center=<4.3596 1.7801 -0.2218>, rotation=<rot 0 90 90>}]
 dyn_skirt_B_master_l [{shapeType=Capsule, length=0.505, radius=0.505, rotation=<rot 0 -90 0>}]
 dyn_skirt_B_master_r [{shapeType=Capsule, length=0.505, radius=0.505, rotation=<rot 0 -90 0>}]
 dyn_skirt_C_01_l [{shapeType=Box, extentX=1.9933, extentY=4.2998, extentZ=9.3569, center=<5.0555 -0.2049 -1.0218>, rotation=<rot 0 90 90>}]
 dyn_skirt_C_01_r [{shapeType=Box, extentX=2.2219, extentY=4.1898, extentZ=9.4209, center=<4.5273 0.1492 0.3575>, rotation=<rot 0 90 90>}]
 dyn_skirt_C_02_l [{shapeType=Box, extentX=1.6196, extentY=4.8209, extentZ=10.2663, center=<4.1537 0.1453 -1.261>, rotation=<rot 0 90 90>}]
 dyn_skirt_C_master_l [{shapeType=Capsule, length=0.505, radius=0.505, rotation=<rot 0 -90 0>}]
 dyn_skirt_C_master_r [{shapeType=Capsule, length=0.505, radius=0.505, rotation=<rot 0 -90 0>}]
 thigh_l [{shapeType=Capsule, length=31.828, radius=9.0305, center=<-22 0 0>, rotation=<rot 90 0 0>}]
 thigh_r [{shapeType=Capsule, length=31.828, radius=9.0305, center=<22 0 0>, rotation=<rot 90 0 0>}]
}
constraints [
 {
  bone1Name=calf_l
  bone2Name=thigh_l
  swing1LimitDegrees=45
  swing1Motion=Free
  swing2LimitDegrees=45
  swing2Motion=Free
  twistLimitDegrees=45
  twistMotion=Free
 }
 {
  bone1Name=calf_r
  bone2Name=thigh_r
  swing1LimitDegrees=45
  swing1Motion=Free
  swing2LimitDegrees=45
  swing2Motion=Free
  twistLimitDegrees=45
  twistMotion=Free
 }
 {bone1Name=dyn_skirt_C_01_r, bone2Name=dyn_skirt_C_master_r, swing1LimitDegrees=60, swing2LimitDegrees=60, twistLimitDegrees=10}
 {bone1Name=dyn_skirt_B_03_r, bone2Name=dyn_skirt_B_02_r, swing1LimitDegrees=60, swing2LimitDegrees=60, twistLimitDegrees=10}
 {bone1Name=dyn_skirt_B_01_r, bone2Name=dyn_skirt_B_master_r, swing1LimitDegrees=60, swing2LimitDegrees=60, twistLimitDegrees=10}
 {bone1Name=dyn_skirt_B_02_r, bone2Name=dyn_skirt_B_01_r, swing1LimitDegrees=60, swing2LimitDegrees=60, twistLimitDegrees=10}
 {bone1Name=dyn_skirt_A_03_r, bone2Name=dyn_skirt_A_02_r, swing1LimitDegrees=60, swing2LimitDegrees=60, twistLimitDegrees=10}
 {bone1Name=dyn_skirt_A_02_r, bone2Name=dyn_skirt_A_01_r, swing1LimitDegrees=60, swing2LimitDegrees=60, twistLimitDegrees=10}
 {bone1Name=dyn_skirt_A_01_r, bone2Name=dyn_skirt_A_master_r, swing1LimitDegrees=60, swing2LimitDegrees=60, twistLimitDegrees=10}
 {bone1Name=dyn_skirt_A_03_l, bone2Name=dyn_skirt_A_02_l, swing1LimitDegrees=60, swing2LimitDegrees=60, twistLimitDegrees=10}
 {bone1Name=dyn_skirt_A_01_l, bone2Name=dyn_skirt_A_master_l, swing1LimitDegrees=60, swing2LimitDegrees=60, twistLimitDegrees=10}
 {bone1Name=dyn_skirt_A_02_l, bone2Name=dyn_skirt_A_01_l, swing1LimitDegrees=60, swing2LimitDegrees=60, twistLimitDegrees=10}
 {bone1Name=dyn_skirt_B_03_l, bone2Name=dyn_skirt_B_02_l, swing1LimitDegrees=60, swing2LimitDegrees=60, twistLimitDegrees=10}
 {bone1Name=dyn_skirt_B_02_l, bone2Name=dyn_skirt_B_01_l, swing1LimitDegrees=60, swing2LimitDegrees=60, twistLimitDegrees=10}
 {bone1Name=dyn_skirt_B_01_l, bone2Name=dyn_skirt_B_master_l, swing1LimitDegrees=60, swing2LimitDegrees=60, twistLimitDegrees=10}
 {bone1Name=dyn_skirt_C_02_l, bone2Name=dyn_skirt_C_01_l, swing1LimitDegrees=60, swing2LimitDegrees=60, twistLimitDegrees=10}
 {bone1Name=dyn_skirt_C_01_l, bone2Name=dyn_skirt_C_master_l, swing1LimitDegrees=60, swing2LimitDegrees=60, twistLimitDegrees=10}
]
