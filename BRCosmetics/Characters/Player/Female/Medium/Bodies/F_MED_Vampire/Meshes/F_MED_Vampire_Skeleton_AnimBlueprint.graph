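# Animation Blueprint for the female Vampire body cosmetic, adding deformation and secondary motion on the character part.
# It copies the attached parent mesh's pose, runs the body's Control Rig and caches the result.
# With cloth enabled it outputs that pose and updates a parameter-asset cloth controller.
# Otherwise it outputs a pelvis-space RigidBody simulation of that pose and stores computed gravity override and rigid-body settings.
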AnimBlueprint /BRCosmetics/Characters/Player/Female/Medium/Bodies/F_MED_Vampire/Meshes/F_MED_Vampire_Skeleton_AnimBlueprint
target_skeleton=&/F_MED_Vampire_Skeleton
graphs [
 AnimGraph {nodes [
   n1 = RigidBody!(
    ComponentPose=Local To Component(LocalPose=Use cached pose 'Post Deform'())
    Alpha
    OverrideWorldGravity:Vector="0,0,-980"
    ExternalForce
    SimSpaceSettings:Sim Space Settings Structure=(WorldAlpha=0,VelocityScaleZ=1,DampingAlpha=1,MaxLinearVelocity=10000,MaxAngularVelocity=10000,MaxLinearAcceleration=10000,MaxAngularAcceleration=10000,ExternalLinearDragV=(X=0,Y=0,Z=0),ExternalLinearVelocity=(X=0,Y=0,Z=0),ExternalAngularVelocity=(X=0,Y=0,Z=0))
   ) -> (Pose) {
    bOverrideWorldGravity=true
    overridePhysicsAsset=&/F_MED_Vampire_Physics
    simulationSpace=BaseBoneSpace
    alphaBoolBlend {blendOption=Linear}
    baseBoneRef {boneName=pelvis}
    overrideWorldGravity {z=-980}
   }
   Output Pose(Result=Blend Poses by bool(BlendPose_0=Use cached pose 'Post Deform'(), BlendPose_1=Component To Local(ComponentPose=@n1), bActiveValue=Get bClothEnabled()) {blendType=Linear}) {layerGroup=None}
   n2 = Control Rig(Source=Copy Pose From Mesh() {bUseAttachedParent=true}) {
    alphaBoolBlend {blendOption=Linear}
    controlRigAssetReference {blueprintRigClass=&/F_MED_Vampire_CtrlRig.F_MED_Vampire_CtrlRig_C}
   }
   Save cached pose 'Post Deform'(Pose=@n2) {cacheName=Post Deform}
  ]}
 EventGraph {nodes [
   n1 = Event BlueprintUpdateAnimation()
   > Set Delta Time X(Delta Time X=@n1.DeltaTimeX)
   > n2 = Branch(Condition=Get bClothEnabled())
   > "Cloth Controller Function (Parameter Asset) (Nativized)"!(
    execute
    CharacterPartAnimInstance:Custom Character Part Anim Instance Object Reference=Self-Reference()
    ClothingAssetName:Name=F_MED_Vampire_Clothing_0
    ClothCoreSettings:"Cloth Core Settings Structure (by ref)"=Get Cloth Core Settings() -> (Cloth Core Settings:Cloth Core Settings Structure)
    WindGustLocalSettings:"Wind Gust Settings Structure (by ref)"=Get Wind Gust Local Settings() -> (Wind Gust Local Settings:Wind Gust Settings Structure)
    ClothUpAxis:Vector="1,0,0"
    LinVelScaleFwdBwdClamp:float=0
    LinVelScaleUpDwnClamp:float=0.7
    DeltaTime:float=Get Delta Time X()
    GameTimeInSeconds:float=GetGameTimeInSeconds()
    StateLogicLayerName:Name=none
    Parameters:Fort Character Dynamics Parameters Object Reference=&/F_MED_Vampire_Parameters
   ) -> (
    then
    OutCurrentState:Name=None
    OutClothParameters:Cloth Parameters Structure
    OutClothCoreSettings:Cloth Core Settings Structure
    OutWindGustSettings:Wind Gust Parameters Structure
    OutWindGustLocalSettings:Wind Gust Settings Structure
    OutWindVelocity:Vector
    OutLinearVelVector:Vector
   )
   n3 = "Gravity Override State Function (Parameter Asset) (Nativized)"(
    execute=@n2.else
    CharacterPartAnimInstance=Self-Reference()
    GameTimeInSeconds=GetGameTimeInSeconds()
    Parameters=&F_MED_Vampire_Parameters
   )
   > Set Gravity Override(Gravity Override=@n3)
   > Set RBAN Settings Out(RBAN Settings Out=@n3.OutRigidBodyAnimNodeParameters)
  ]}
]
variables [
 Cloth Core Settings {
  type=ClothCoreSettings
  default {
   collisionThickness=1
   dampingCoefficient=0
   frictionCoefficient=0.3
   localDampingCoefficient=0.1
   numberIterations=2
   numberSubsteps=1
   selfCollisionThickness=0
   animDriveDamping=<0 1>
   animDriveStiffness=<0 1>
   areaStiffness=<1 1>
   bendingStiffness=<1 1>
   edgeStiffness=<1 1>
   tetherStiffness=<1 1>
  }
 }
 Delta Time X {type=number}
 Gravity Override {type=Vector}
 RBAN Settings Out {type=FortRigidBodyAnimNodeParameters}
 Wind Gust Local Settings {type=WindGustSettings, default {elapsedTime=0, finalGustStrengthLocal=0, gustDuration=0, gustStrength=0, timeBetweenGusts=0}}
]
