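# Control Rig for the female medium Vampire character's hair, adjusting the back hair base bones as the head moves.
# Sequence A chains aim constraints that orient the left and right hair A and B bases against spine_05.
# Sequence B uses a spherical pose reader on head relative to spine_05, with debug drawing on.
# Its output weights corrective transform offsets on hair bones, including the left A and B bases.
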ControlRigBlueprint /BRCosmetics/Characters/Player/Female/Medium/Bodies/F_MED_Vampire/Meshes/Parts/F_MED_Vampire_Hair_CtrlRig
graphs [
 RigVMFunctionLibrary {links [], nodes []}
 RigVMModel {nodes [
   RigUnit_BeginExecution = &RigUnit_BeginExecution()
   RigVMFunction_Sequence = &RigVMFunction_Sequence(ExecuteContext=@RigUnit_BeginExecution.ExecutePin)
   &RigUnit_AimConstraintLocalSpaceOffset(
    Child=(Type=Bone,Name="hair_A_01_l_base")
    Filter=(bX=True,bY=True,bZ=True)
    Parents=((Item=(Type=Bone,Name="spine_05"),Weight=1))
    Weight=1
    WorldUp=(Target=(X=0,Y=0,Z=1),Kind=Location,Space=(Type=Bone,Name="spine_05"))
    ExecutePin=@RigVMFunction_Sequence.A
   )
   > &RigUnit_AimConstraintLocalSpaceOffset(
    Child=(Type=Bone,Name="hair_B_01_l_base")
    Filter=(bX=True,bY=True,bZ=True)
    Parents=((Item=(Type=Bone,Name="spine_05"),Weight=1))
    Weight=1
    WorldUp=(Target=(X=0,Y=0,Z=1),Kind=Location,Space=(Type=Bone,Name="spine_05"))
   )
   > &RigUnit_AimConstraintLocalSpaceOffset(
    Child=(Type=Bone,Name="hair_A_01_r_base")
    Filter=(bX=True,bY=True,bZ=True)
    Parents=((Item=(Type=Bone,Name="spine_05"),Weight=1))
    Weight=1
    WorldUp=(Target=(X=0,Y=0,Z=1),Kind=Location,Space=(Type=Bone,Name="spine_05"))
   )
   > &RigUnit_AimConstraintLocalSpaceOffset(
    Child=(Type=Bone,Name="hair_B_01_r_base")
    Filter=(bX=True,bY=True,bZ=True)
    Parents=((Item=(Type=Bone,Name="spine_05"),Weight=1))
    Weight=1
    WorldUp=(Target=(X=0,Y=0,Z=1),Kind=Location,Space=(Type=Bone,Name="spine_05"))
   )
   n1 = &RigUnit_SphericalPoseReader(
    ActiveRegionScaleFactors=(PositiveWidth=3,NegativeWidth=1,PositiveHeight=3,NegativeHeight=3)
    ActiveRegionSize=0.1
    Debug=(bDrawDebug=True,bDraw2D=False,bDrawLocalAxes=False,DebugScale=25,DebugSegments=20,DebugThickness=0.25)
    DriverAxis=(X=1,Y=0,Z=0)
    DriverItem=(Type=Bone,Name="head")
    FalloffRegionScaleFactors=(PositiveWidth=1,NegativeWidth=1,PositiveHeight=1,NegativeHeight=1)
    FalloffSize=0.3
    FlipHeightScaling=False
    FlipWidthScaling=False
    OptionalParentItem=(Type=Bone,Name="spine_05")
    RotationOffset=(X=90,Y=0,Z=-145)
    ExecutePin=@RigVMFunction_Sequence.B
   )
   > &RigUnit_ModifyTransforms(
    ItemToModify="((Item=(Type=Bone,Name=\"hair_A_01_l_base\"),Transform=(Rotation=(X=0,Y=0.2588,Z=0,W=0.9659),Translation=(X=0,Y=0,Z=0),Scale3D=(X=1,Y=1,Z=1))),(Item=(Type=Bone,Name=\"hair_B_01_l_base\"),Transform=(Rotation=(X=-0.0302,Y=0.171,Z=0.171,W=0.9698),Translation=(X=0 ...(truncated)"
    WeightMinimum=0
    Weight=@n1.OutputParam
   )
  ]}
]
hierarchy [
 root {loc=<0 0 0>, rot=<0 0 0 1>}
 pelvis {parent=root, loc=<0 -0.5301 98.0571>, rot=<0.0127 0.707 -0.0127 -0.707>}
 spine_01 {parent=pelvis, loc=<6.1208 -0.9328 0>, rot=<0 0 0.034 -0.9994>}
 spine_02 {parent=spine_01, loc=<6.5996 -0.2095 0>, rot=<0 0 0.0238 -0.9997>}
 spine_03 {parent=spine_02, loc=<6.4914 -0.1237 0>, rot=<0 0 0.0149 -0.9999>}
 spine_04 {parent=spine_03, loc=<6.9598 0.0794 0>, rot=<0 0 -0.0132 -0.9999>}
 spine_05 {parent=spine_04, loc=<6.9987 0 0>, rot=<0 0 -0.0015 -1>}
 neck_01 {parent=spine_05, loc=<11.3138 0.1529 0>, rot=<0 0 -0.2003 -0.9797>}
 neck_02 {parent=neck_01, loc=<4.5627 0 0>, rot=<0 0 0.0504 -0.9987>}
 head {parent=neck_02, loc=<5.3914 0 0>, rot=<0 0 0.115 -0.9934>}
 ponytail_base {parent=head, loc=<13.7731 -4.415 0>, rot=<0 0 0.369 -0.9294>}
 ponytail_A {parent=ponytail_base, loc=<5.5025 0 0>, rot=<0 0 0.2842 -0.9588>}
 ponytail_B {parent=ponytail_base, loc=<2.8026 -4.7511 0>, rot=<0 0 0.7391 -0.6736>}
 ponytail_C {parent=ponytail_base, loc=<5.5025 0 -0.9029>, rot=<-0.0878 -0.1286 0.2988 -0.9415>}
 bangs_01_l_base {parent=head, loc=<6.2674 1.8658 -6.6409>, rot=<0 0 1 -0.0041>}
 bangs_01_l {parent=bangs_01_l_base, loc=<0.3308 0 0>, rot=<0 0 0 1>}
 bangs_02_l {parent=bangs_01_l, loc=<5.566 0.0179 0.3536>, rot=<0 0 0 1>}
 bangs_03_l {parent=bangs_02_l, loc=<3.415 0.3089 0.2801>, rot=<0 0 0 1>}
 bangs_04_l {parent=bangs_03_l, loc=<4.279 -0.4215 -0.1928>, rot=<0 0 0.1665 -0.986>}
 hair_B_01_l_base {parent=head, loc=<0.0989 -4.3698 -5.0843>, rot=<-0.0528 0.2005 0.9783 -0.0049>}
 hair_B_01_l {parent=hair_B_01_l_base, loc=<0.3308 0 0>, rot=<0 0 0 1>}
 hair_B_02_l {parent=hair_B_01_l, loc=<3.9683 0.1449 0.4815>, rot=<-0.0029 0.0418 0.0695 -0.9967>}
 hair_A_01_l_base {parent=head, loc=<2.3548 -7.0355 -2.3874>, rot=<-0.0027 0.6469 0.7626 -0.0032>}
 hair_A_01_l {parent=hair_A_01_l_base, loc=<0.3308 0 0>, rot=<0 0 0 1>}
 hair_A_02_l {parent=hair_A_01_l, loc=<5.88 0 0>, rot=<0.6581 0 0 -0.7529>}
 bangs_01_r_base {parent=head, loc=<6.2676 1.8658 6.6409>, rot=<0 0 -0.0041 -1>}
 bangs_01_r {parent=bangs_01_r_base, loc=<-0.3308 0 0>, rot=<0 -1 0 0>}
 bangs_02_r {parent=bangs_01_r, loc=<5.566 -0.0179 0.3536>, rot=<0 0 0 1>}
 bangs_03_r {parent=bangs_02_r, loc=<3.4152 -0.3089 0.2801>, rot=<0 0 0 1>}
 bangs_04_r {parent=bangs_03_r, loc=<4.2791 0.4215 -0.1928>, rot=<0 0 0 1>}
 hair_B_01_r_base {parent=head, loc=<0.0989 -4.3698 5.0843>, rot=<-0.2005 -0.0528 -0.0049 -0.9783>}
 hair_B_01_r {parent=hair_B_01_r_base, loc=<-0.3308 0 0>, rot=<0 -1 0 0>}
 hair_B_02_r {parent=hair_B_01_r, loc=<3.9685 -0.1449 0.4815>, rot=<-0.0029 0.0418 0.0695 -0.9967>}
 hair_A_01_r_base {parent=head, loc=<2.355 -7.0355 2.3874>, rot=<-0.6469 -0.0027 -0.0032 -0.7626>}
 hair_A_01_r {parent=hair_A_01_r_base, loc=<-0.331 0 0>, rot=<0 -1 0 0>}
 hair_A_02_r {parent=hair_A_01_r, loc=<5.88 0 0>, rot=<0.6581 0 0 -0.7529>}
]
