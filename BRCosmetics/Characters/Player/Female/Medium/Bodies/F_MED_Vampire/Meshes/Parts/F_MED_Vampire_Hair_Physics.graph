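# Physics asset for the hair mesh part of the medium female Vampire player cosmetic body, defining capsule bodies for bangs, ponytail, head, neck and clavicles.
# Left and right bangs form four-link chains from the head, with swing limits widening toward the tips.
# The three ponytail bodies each constrain directly to the head rather than chaining, and the head constrains to the neck.
# The larger clavicle capsules have no constraints.
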PhysicsAsset /BRCosmetics/Characters/Player/Female/Medium/Bodies/F_MED_Vampire/Meshes/Parts/F_MED_Vampire_Hair_Physics
bodies {
 bangs_01_l [{shapeType=Capsule, length=3, radius=1.5, shapeName=bangs_01_l_capsule, center=<1 0 0>, rotation=<rot 90 0 0>}]
 bangs_01_r [{shapeType=Capsule, length=3, radius=1.5, shapeName=bangs_01_r_capsule, center=<1 0 0>, rotation=<rot 90 0 0>}]
 bangs_02_l [{shapeType=Capsule, length=3, radius=1.5, shapeName=bangs_02_l_capsule, center=<1 0 0>, rotation=<rot 90 0 0>}]
 bangs_02_r [{shapeType=Capsule, length=3, radius=1.5, shapeName=bangs_02_r_capsule, center=<1 0 0>, rotation=<rot 90 0 0>}]
 bangs_03_l [{shapeType=Capsule, length=3, radius=1.5, shapeName=bangs_03_l_capsule, center=<2 0 0>, rotation=<rot 90 0 0>}]
 bangs_03_r [{shapeType=Capsule, length=3, radius=1.5, shapeName=bangs_03_r_capsule, center=<2 0 0>, rotation=<rot 90 0 0>}]
 bangs_04_l [{shapeType=Capsule, length=5, radius=1.5, shapeName=bangs_04_l_capsule, center=<4 0 0>, rotation=<rot 80 175 0>}]
 bangs_04_r [{shapeType=Capsule, length=5, radius=1.5, shapeName=bangs_04_r_capsule, center=<4 0 0>, rotation=<rot 80 175 0>}]
 clavicle_l [{shapeType=Capsule, length=9.5, radius=7.5, center=<2 0.0001 -3.2908>, rotation=<rot 0 90 90>}]
 clavicle_r [{shapeType=Capsule, length=9.5, radius=7.5, center=<-1.9567 -0.0414 3.6248>, rotation=<rot 0 90 90>}]
 head [{shapeType=Capsule, length=10.3851, radius=5.6246, center=<0.0008 0.0444 -0.3098>, rotation=<rot 90 0 0>}]
 neck_01 [{shapeType=Capsule, length=6.6676, radius=3.6177, center=<-0.3571 1.6165 0>, rotation=<rot 0 -90 0>}]
 ponytail_A [{shapeType=Capsule, length=2.1646, radius=3.5994, center=<3.0785 0.0203 3.1583>, rotation=<rot 90 0 0>}]
 ponytail_B [{shapeType=Capsule, length=0, radius=4.6265, center=<6.2366 -0.1803 -0.0406>, rotation=<rot 0 -90 0>}]
 ponytail_C [{shapeType=Capsule, length=2.1646, radius=3.5994, center=<3.9022 -0.1103 -2.2108>, rotation=<rot 90 0 0>}]
}
constraints [
 {bone1Name=ponytail_A, bone2Name=head, swing1LimitDegrees=40, swing2LimitDegrees=40, twistLimitDegrees=10}
 {bone1Name=ponytail_B, bone2Name=head, swing1LimitDegrees=40, swing2LimitDegrees=40, twistLimitDegrees=10}
 {bone1Name=ponytail_C, bone2Name=head, swing1LimitDegrees=40, swing2LimitDegrees=40, twistLimitDegrees=10}
 {bone1Name=head, bone2Name=neck_01, swing1LimitDegrees=40, swing2LimitDegrees=40, twistLimitDegrees=10}
 {bone1Name=bangs_01_r, bone2Name=head, swing1LimitDegrees=10, swing2LimitDegrees=10, twistLimitDegrees=5}
 {bone1Name=bangs_02_r, bone2Name=bangs_01_r, swing1LimitDegrees=10, swing2LimitDegrees=10, twistLimitDegrees=5}
 {bone1Name=bangs_04_r, bone2Name=bangs_03_r, swing1LimitDegrees=20, swing2LimitDegrees=20, twistLimitDegrees=5}
 {bone1Name=bangs_01_l, bone2Name=head, swing1LimitDegrees=10, swing2LimitDegrees=10, twistLimitDegrees=5}
 {bone1Name=bangs_02_l, bone2Name=bangs_01_l, swing1LimitDegrees=10, swing2LimitDegrees=10, twistLimitDegrees=5}
 {bone1Name=bangs_04_l, bone2Name=bangs_03_l, swing1LimitDegrees=20, swing2LimitDegrees=20, twistLimitDegrees=5}
 {bone1Name=bangs_03_l, bone2Name=bangs_02_l, swing1LimitDegrees=15, swing2LimitDegrees=15, twistLimitDegrees=5}
 {bone1Name=bangs_03_r, bone2Name=bangs_02_r, swing1LimitDegrees=15, swing2LimitDegrees=15, twistLimitDegrees=5}
]
