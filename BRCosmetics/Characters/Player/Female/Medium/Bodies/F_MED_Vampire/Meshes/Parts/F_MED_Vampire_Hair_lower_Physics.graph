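# Physics asset for the lower hair of the medium female Vampire player body, using capsule bodies for hair physics.
# Four chains (A and B, left and right) run from small base capsules on the head through two segments with tight swing and twist limits.
# Angularly free constraints attach the bases to the head and cross-link each side's A and B tips.
# Head, neck, upper spine and clavicle capsules supply the body shapes.
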PhysicsAsset /BRCosmetics/Characters/Player/Female/Medium/Bodies/F_MED_Vampire/Meshes/Parts/F_MED_Vampire_Hair_lower_Physics
bodies {
 clavicle_l [{shapeType=Capsule, length=12.3544, radius=5.111, center=<2.5742 -1.9936 -0.6187>, rotation=<rot 0 90 90>}]
 clavicle_r [{shapeType=Capsule, length=12.1319, radius=4.9801, center=<-3.3097 1.9199 0.786>, rotation=<rot 0 90 90>}]
 hair_A_01_l [{shapeType=Capsule, length=5, radius=2.25, shapeName=hair_A_01_l_capsule, center=<5 0 0>, rotation=<rot 90 0 0>}]
 hair_A_01_l_base [{shapeType=Capsule, length=0.505, radius=0.505, shapeName=hair_A_01_l_base_capsule, rotation=<rot 0 -90 0>}]
 hair_A_01_r [{shapeType=Capsule, length=6.0937, radius=2.5897, shapeName=hair_A_01_r_capsule, center=<7 0 0>, rotation=<rot 90 0 0>}]
 hair_A_01_r_base [{shapeType=Capsule, length=0.505, radius=0.505, shapeName=hair_A_01_r_base_capsule, rotation=<rot 0 -90 0>}]
 hair_A_02_l [{shapeType=Capsule, length=5.2012, radius=2.7039, shapeName=hair_A_02_l_capsule, center=<6.8921 -1 -2>, rotation=<rot 0 -10.2688 93.1478>}]
 hair_A_02_r [{shapeType=Capsule, length=5.2012, radius=2.7039, shapeName=hair_A_02_r_capsule, center=<6.8916 1 -2.5>, rotation=<rot 0 -12.0637 97.1394>}]
 hair_B_01_l [{shapeType=Capsule, length=5, radius=2, shapeName=hair_B_01_l_capsule, center=<4 0 -1>, rotation=<rot 90 0 0>}]
 hair_B_01_l_base [{shapeType=Capsule, length=0.505, radius=0.505, shapeName=hair_B_01_l_base_capsule, rotation=<rot 0 -90 0>}]
 hair_B_01_r [{shapeType=Capsule, length=5.6669, radius=1.7438, shapeName=hair_B_01_r_capsule, center=<5 0 -1>, rotation=<rot 85 0.0002 0>}]
 hair_B_01_r_base [{shapeType=Capsule, length=0.505, radius=0.505, shapeName=hair_B_01_r_base_capsule, rotation=<rot 0 -90 0>}]
 hair_B_02_l [{shapeType=Capsule, length=4.9396, radius=2.2987, shapeName=hair_B_02_l_capsule, center=<4.9747 0 -2>, rotation=<rot 0.2917 13.85 101.0298>}]
 hair_B_02_r [{shapeType=Capsule, length=5.4041, radius=2, shapeName=hair_B_02_r_capsule, center=<6.4987 0.17 -2>, rotation=<rot 0 17.1539 89.9998>}]
 head [{shapeType=Capsule, length=10.3851, radius=5.6246, center=<0.0008 0.0444 -0.3098>, rotation=<rot 90 0 0>}]
 neck_01 [{shapeType=Capsule, length=4.5, radius=4, rotation=<rot 90 0 0>}]
 spine_05 [{shapeType=Capsule, length=21.5625, radius=9.4688, shapeName=spine_05_capsule, center=<0.5135 0.691 0>, rotation=<rot 0 -90 0>}]
}
constraints [
 {bone1Name=head, bone2Name=neck_01, swing1LimitDegrees=40, swing2LimitDegrees=40, twistLimitDegrees=10}
 {bone1Name=hair_A_01_r, bone2Name=hair_A_01_r_base, swing1LimitDegrees=12, swing2LimitDegrees=20, twistLimitDegrees=5}
 {bone1Name=hair_A_02_r, bone2Name=hair_A_01_r, swing1LimitDegrees=12, swing2LimitDegrees=20, twistLimitDegrees=5}
 {bone1Name=hair_B_01_r, bone2Name=hair_B_01_r_base, swing1LimitDegrees=12, swing2LimitDegrees=20, twistLimitDegrees=5}
 {bone1Name=hair_A_01_l, bone2Name=hair_A_01_l_base, swing1LimitDegrees=12, swing2LimitDegrees=20, twistLimitDegrees=5}
 {bone1Name=hair_A_02_l, bone2Name=hair_A_01_l, swing1LimitDegrees=12, swing2LimitDegrees=20, twistLimitDegrees=5}
 {bone1Name=hair_B_01_l, bone2Name=hair_B_01_l_base, swing1LimitDegrees=12, swing2LimitDegrees=20, twistLimitDegrees=5}
 {bone1Name=hair_B_02_l, bone2Name=hair_B_01_l, swing1LimitDegrees=12, swing2LimitDegrees=20, twistLimitDegrees=5}
 {bone1Name=hair_B_02_r, bone2Name=hair_B_01_r, swing1LimitDegrees=12, swing2LimitDegrees=20, twistLimitDegrees=5}
 {
  bone1Name=hair_A_01_l_base
  bone2Name=head
  swing1LimitDegrees=45
  swing1Motion=Free
  swing2LimitDegrees=45
  swing2Motion=Free
  twistLimitDegrees=45
  twistMotion=Free
 }
 {
  bone1Name=hair_A_01_r_base
  bone2Name=head
  swing1LimitDegrees=45
  swing1Motion=Free
  swing2LimitDegrees=45
  swing2Motion=Free
  twistLimitDegrees=45
  twistMotion=Free
 }
 {
  bone1Name=hair_B_01_l_base
  bone2Name=head
  swing1LimitDegrees=45
  swing1Motion=Free
  swing2LimitDegrees=45
  swing2Motion=Free
  twistLimitDegrees=45
  twistMotion=Free
 }
 {
  bone1Name=hair_B_01_r_base
  bone2Name=head
  swing1LimitDegrees=45
  swing1Motion=Free
  swing2LimitDegrees=45
  swing2Motion=Free
  twistLimitDegrees=45
  twistMotion=Free
 }
 {
  bone1Name=neck_01
  bone2Name=spine_05
  swing1LimitDegrees=45
  swing1Motion=Free
  swing2LimitDegrees=45
  swing2Motion=Free
  twistLimitDegrees=45
  twistMotion=Free
 }
 {
  bone1Name=hair_B_02_l
  bone2Name=hair_A_02_l
  swing1LimitDegrees=45
  swing1Motion=Free
  swing2LimitDegrees=45
  swing2Motion=Free
  twistLimitDegrees=45
  twistMotion=Free
 }
 {
  bone1Name=hair_B_02_r
  bone2Name=hair_A_02_r
  swing1LimitDegrees=45
  swing1Motion=Free
  swing2LimitDegrees=45
  swing2Motion=Free
  twistLimitDegrees=45
  twistMotion=Free
 }
]
