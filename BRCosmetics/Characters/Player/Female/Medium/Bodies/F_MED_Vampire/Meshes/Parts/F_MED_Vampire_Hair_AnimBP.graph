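# Animation Blueprint for the female Vampire character's hair, adding procedural and physics motion on top of the attached parent mesh's pose.
# The copied pose runs through a hair Control Rig, then two RigidBody simulations based at head and spine_05.
# Skydiving events set variables feeding the lower simulation's gravity and sim-space settings, flipping gravity upward during the dive.
# A chain of additive hair bone offsets is left unconnected to the output.
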AnimBlueprint /BRCosmetics/Characters/Player/Female/Medium/Bodies/F_MED_Vampire/Meshes/Parts/F_MED_Vampire_Hair_AnimBP
target_skeleton=&/F_MED_Vampire_Hair_Skeleton
graphs [
 AnimGraph {nodes [
   n1 = "Transform (Modify) Bone - Bone: hair_A_01_r"(Alpha=1, Translation="0,0,0", Rotation="0.000,0.000,10", Scale="1,1,1") {rotationMode=BMM_Additive, alphaBoolBlend {blendOption=Linear}, boneToModify {boneName=hair_A_01_r}, rotation {roll=10}}
   n2 = "Transform (Modify) Bone - Bone: hair_A_02_r"(ComponentPose=@n1, Alpha=1, Translation="0,0,0", Rotation="0,0.000,20", Scale="1,1,1") {rotationMode=BMM_Additive, alphaBoolBlend {blendOption=Linear}, boneToModify {boneName=hair_A_02_r}, rotation {roll=20}}
   n3 = "Transform (Modify) Bone - Bone: hair_A_01_l"(ComponentPose=@n2, Alpha=1, Translation="0,0,0", Rotation="0.000,0.000,10", Scale="1,1,1") {rotationMode=BMM_Additive, alphaBoolBlend {blendOption=Linear}, boneToModify {boneName=hair_A_01_l}, rotation {roll=10}}
   n4 = "Transform (Modify) Bone - Bone: hair_B_01_r"(ComponentPose=@n3, Alpha=1, Translation="0,0,0", Rotation="0.000,0,10.000", Scale="1,1,1") {rotationMode=BMM_Additive, alphaBoolBlend {blendOption=Linear}, boneToModify {boneName=hair_B_01_r}, rotation {roll=10}}
   "Transform (Modify) Bone - Bone: hair_B_02_l"(
    ComponentPose="Transform (Modify) Bone - Bone: hair_B_01_l"(
     ComponentPose="Transform (Modify) Bone - Bone: hair_B_02_r"(ComponentPose=@n4, Alpha=1, Translation="0,0,0", Rotation="0,0,0", Scale="1,1,1") {rotationMode=BMM_Additive, alphaBoolBlend {blendOption=Linear}, boneToModify {boneName=hair_B_02_r}}
     Alpha=1
     Translation="0,0,0"
     Rotation="0,0,0"
     Scale="1,1,1"
    ) {rotationMode=BMM_Additive, alphaBoolBlend {blendOption=Linear}, boneToModify {boneName=hair_B_01_l}}
    Alpha=1
    Translation="0,0,3"
    Rotation="0,0,-20"
    Scale="1,1,1"
   ) {
    rotationMode=BMM_Additive
    translationMode=BMM_Additive
    alphaBoolBlend {blendOption=Linear}
    boneToModify {boneName=hair_B_02_l}
    rotation {roll=-20}
    translation {z=3}
   }
   n5 = Control Rig(Source=Copy Pose From Mesh() {bUseAttachedParent=true}) {
    alphaBoolBlend {blendOption=Linear}
    controlRigAssetReference {blueprintRigClass=&/F_MED_Vampire_Hair_CtrlRig.F_MED_Vampire_Hair_CtrlRig_C}
   }
   n6 = RigidBody(ComponentPose=Local To Component(LocalPose=@n5)) {
    bClampLinearTranslationLimitToRefPose=true
    bOverrideWorldGravity=true
    overridePhysicsAsset=&/F_MED_Vampire_Hair_Physics
    simulationSpace=BaseBoneSpace
    alphaBoolBlend {blendOption=Linear}
    baseBoneRef {boneName=head}
    overrideWorldGravity {z=-980}
    simSpaceSettings {
     dampingAlpha=0
     maxAngularAcceleration=20
     maxAngularVelocity=4
     maxLinearAcceleration=3500
     maxLinearVelocity=770
     velocityScaleZ=0.8
     worldAlpha=0.5
     externalLinearDragV=<2 2 2>
    }
   }
   n7 = RigidBody!(
    ComponentPose=@n6
    Alpha
    OverrideWorldGravity:Vector=Get Override World Gravity() -> (Override World Gravity:Vector)
    ExternalForce
    SimSpaceSettings:Sim Space Settings Structure=Get Sim Space Settings() -> (Sim Space Settings:Sim Space Settings Structure)
   ) -> (Pose) {
    bClampLinearTranslationLimitToRefPose=true
    bOverrideWorldGravity=true
    overridePhysicsAsset=&/F_MED_Vampire_Hair_lower_Physics
    simulationSpace=BaseBoneSpace
    alphaBoolBlend {blendOption=Linear}
    baseBoneRef {boneName=spine_05}
    overrideWorldGravity {z=-980}
    simSpaceSettings {dampingAlpha=0, maxAngularAcceleration=30, maxAngularVelocity=300, maxLinearAcceleration=4000, maxLinearVelocity=770, worldAlpha=1}
   }
   Output Pose(Result=Component To Local(ComponentPose=@n7)) {layerGroup=None}
  ]}
 EventGraph {nodes [
   Event OnBeginSkydiving()
   > Set Override World Gravity(Override World Gravity:Vector="0,0,450") -> (Output_Get:Vector)
   > Set Sim Space Settings(Sim Space Settings:Sim Space Settings Structure) -> (Output_Get:Sim Space Settings Structure)
   Event OnEndSkydiving()
   > Set Override World Gravity(Override World Gravity="0,0,-980")
   > Set Sim Space Settings()
   Event BlueprintUpdateAnimation()
  ]}
]
variables [
 Override World Gravity {type=Vector, default=<0 0 -980>}
 Sim Space Settings {
  type=SimSpaceSettings
  default {
   dampingAlpha=0
   maxAngularAcceleration=30
   maxAngularVelocity=100
   maxLinearAcceleration=2000
   maxLinearVelocity=770
   velocityScaleZ=0.5
   worldAlpha=0.5
   externalLinearDragV=<1 1 1>
  }
 }
]
